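; Marker actor for the Tidal Crane boss's falling rock attack that shows a Niagara warning tell for a rock's landing spot.
; A custom event stores the ground trace location in a variable with a RepNotify.
; The notify passes that location to the tell's "EndPoint" parameter only where cosmetic events can execute.
; On EndPlay, it calls Activate on the tell without reset, then immediately destroys that component.
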(event EventBeginPlay)

(event EventEndPlay
  (params (EndPlayReason :type EEndPlayReason%20Enum :default "Destroyed"))
  (bind _ns_tidalcrane_rockfallingtell (Variables|Default|GetNS_TidalCrane_RockFallingTell))
  (Components|Activation|Activate _ns_tidalcrane_rockfallingtell false)
  (Components|DestroyComponent _ns_tidalcrane_rockfallingtell))

(event Custom|SetGroundTraceLocation
  (params (Location :type Vector))
  (Variables|Default|SetGroundTraceLocation Location))

(fn OnRep_GroundTraceLocation
  (Networking|CanExecuteCosmeticEvents
    (:True
      (Niagara|SetNiagaraVariable(Position) (Variables|Default|GetNS_TidalCrane_RockFallingTell) "EndPoint" (Variables|Default|GetGroundTraceLocation)))))

(fn ConstructionScript)
