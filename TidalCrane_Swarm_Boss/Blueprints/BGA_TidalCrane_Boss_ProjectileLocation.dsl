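; Target-location marker for the TidalCrane boss's projectile, showing the NS_TidalCrane_BossTarget effect.
; After a brief delay on BeginPlay, it casts BossActor to BGA_TidalCrane_Boss and turns ShowDecal off if the boss mesh's world X scale is at least six.
; The ShowDecal RepNotify passes the flag to the effect's parameter.
; EndPlay calls Activate on the effect component without reset, then destroys it.
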(event EventBeginPlay
  (Utilities|FlowControl|Delay 0.033)
  (bind _asbga_tidal_crane_boss (Utilities|Casting|CastToBGA_TidalCrane_Boss (Variables|Default|GetBossActor)))
  (if (>= (.x (Transformation|GetWorldScale (Variables|/TidalCrane_Swarm_Boss/Gameplay/BGA_TidalCrane_Boss.BGA_TidalCrane_Boss_C|GetSkeletalMeshComponentBudgeted _asbga_tidal_crane_boss))) 6.0)
    (Variables|Default|SetShowDecal? false)))

(event EventEndPlay
  (params (EndPlayReason :type EEndPlayReason%20Enum :default "Destroyed"))
  (bind _ns_tidalcrane_bosstarget (Variables|Default|GetNS_TidalCrane_BossTarget))
  (Components|Activation|Activate _ns_tidalcrane_bosstarget false)
  (Components|DestroyComponent _ns_tidalcrane_bosstarget))

(fn OnRep_ShowDecal?
  (Effects|Components|ParticleSystem|SetBoolParameter (Variables|Default|GetNS_TidalCrane_BossTarget) "ShowDecal" (Variables|Default|GetShowDecal?)))

(fn ConstructionScript)
