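; Boss spawner static point provider for the TidalCrane Swarm boss that places the boss base in the world.
; On BeginPlay, if base spawning is allowed, it spawns a boss base class at its own transform, picking between a standard and a no-RVT variant based on whether it sits in a tunnel.
; A custom EnableProvider event enables its Living World point provider and requests event generation from the Living World manager.
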(event EventBeginPlay
  (if (Variables|Default|GetAllowSpawnBase)
    (Game|SpawnActorfromClass (select (Variables|Default|GetIsInTunnel) (Variables|Default|GetBossBaseClassNoRVT) (Variables|Default|GetBossBaseClass)) (Transformation|GetActorTransform) "AlwaysSpawn" "MultiplyWithRoot" 0 0)))

(event Custom|EnableProvider
  (Class|FortAthenaLivingWorldPointProviderInterface|EnablePointProvider_DEPRECATED self)
  (Class|FortAthenaLivingWorldManager|RequestEventGeneration (Class|FortAthenaLivingWorldManager|GetLivingWorldManager)))

(fn ConstructionScript)
