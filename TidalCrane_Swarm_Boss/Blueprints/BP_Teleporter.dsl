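; BP_Teleporter is an overlap-triggered teleporter that moves a Fortnite player pawn to a target actor's location and rotation.
; On BeginPlay, it sets the target reference to itself if none is valid.
; On overlap, it casts the other actor to PlayerPawn_Athena, stores the pawn, and teleports it only if it is neither bot-controlled nor tethered.
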(event EventBeginPlay
  ;[1 1 1] If actor spawn location is not specified, set it to my location
  (Utilities|IsValid (Variables|Default|GetActorLocation)
    (:Is%20Not%20Valid
      (Variables|Default|SetActorLocation self))))

(event Collision|EventActorBeginOverlap
  (params (OtherActor :type Actor%20Object%20Reference))
  (bind _asplayer_pawn_athena (Utilities|Casting|CastToPlayerPawn_Athena OtherActor))
  (bind _output_get (Variables|InternalData|SetPlayerPawnAthena _asplayer_pawn_athena))
  (if (Pawn|IsBotControlled _asplayer_pawn_athena)
    _
    (elif (Class|FortPlayerPawn|IsTethered _asplayer_pawn_athena)
      _
      (else
        (Transform|Actor|TeleportPlayerPawn _output_get (Transformation|GetActorLocation (Variables|Default|GetActorLocation)) (Transformation|GetActorRotation (Variables|Default|GetActorLocation)) true)))))

(fn ConstructionScript)
