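; Spawn manager for the Tidal Crane queen bug boss in the bug biome, tracking summon progress and running the encounter until the boss dies.
; The server reserves AI budget, binds nearby chests and spawners and activates spires; chest openings, spawner destructions and team-size-scaled bug kills add progress.
; SpawnBoss, called after the startup delay and at full progress, frees the budget and enables a random Living World point.
; Clients run music, roar effects and zone-entry UI events.
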(event EventBeginPlay
  (|CallParentFunction)
  (bind (_gamestate _playlist) (Class|FortAsyncActionCurrentPlaylistReady|PlaylistDataReadyAsync)
    (:OnReady
      (Variables|Default|SetTeamSizeProgressMultiplier (ScalableFloat|GetValueatLevel (Utilities|Select :Index (Variables|/Script/FortniteGame.FortPlaylist|GetMaxTeamSize _playlist) :Option%200 (Variables|Default|GetSF_ProgressMultiplierInSolos) :Option%201 (Variables|Default|GetSF_ProgressMultiplierInSolos) :Option%202 (Variables|Default|GetSF_ProgressMultiplierInDuos) :Option%203 (Variables|Default|GetSF_ProgressMultiplierInTrios) :Option%204 (Variables|Default|GetSF_ProgressMultiplierInSquads))))
      (Networking|SwitchHasAuthority
        (:Authority
          (Utilities|FlowControl|DelayUntilNextTick)
          (CallFunction|ReserveAIBudgetforQueen)
          (CallFunction|GetChestsandSpires)
          (if (ScalableFloat|GetValueasBool (Variables|Default|GetEnableMiniMapIcon))
            (Components|Activation|SetActive (Variables|Default|GetAthenaSpecialActor) true false)))
        (:Remote
          (AI|Logic|StartLogic (Variables|Default|GetMusicStateTree))
          (Utilities|FlowControl|DelayUntilNextTick)
          (CallFunction|UpdatePotentialOverlapsAfterInit))))))

(event Custom|ReserveAIBudgetForQueen
  (bind _returnvalue (Class|FortAthenaLivingWorldManager|RegisterDensityReservoir (Class|FortAthenaLivingWorldManager|GetLivingWorldManager) self 1.0 1.2e+04))
  (if (not _returnvalue)
    (Fortnite|FortShippingLog self "::BOSS SPAWN MANAGER : Trying to Reserve 1 AI for Queen Summoing, but the registration FAILED!" true)))

(event Custom|GetChestsAndSpires
  (Utilities|FlowControl|Delay (Variables|Default|GetDelayBeforeAcquiringNearbyActors))
  (Utilities|FlowControl|Sequence
    (:then_0
      (Utilities|FlowControl|DoOnce false
        (:Completed
          (CallFunction|SpawnBoss))))
    (:then_1
      (bind _outactors (Collision|SphereOverlapActors (Transformation|GetActorLocation) (Variables|Default|GetGoldenRewardCheckRadius) (Utilities|Array|MakeArray "ObjectTypeQuery1" "ObjectTypeQuery2")))
      (for _array_element _outactors
        (bind _asalways_spawn_bug_chest_bug_biome (Utilities|Casting|CastToAlwaysSpawn_BugChest_BugBiome _array_element)
          (:then
            (Default|BindEventtoOnOpened _asalways_spawn_bug_chest_bug_biome (EventDispatchers|CreateEvent :!event Custom|OnGoldenRewardOpened)))
          (:CastFailed
            (Utilities|IsA(soft) _array_element "/BugBiomeGameplay/Gameplay/Spawners/BGA_LARGE_ContinuousBugSpawner_BugBiome_Child.BGA_LARGE_ContinuousBugSpawner_BugBiome_Child_C"
              (:True
                (bind _asbga_continuous_bug_spawner_bug_biome (Utilities|Casting|CastToBGA_ContinuousBugSpawner_BugBiome _array_element))
                (Game|BindEventtoOnDestroyed _asbga_continuous_bug_spawner_bug_biome (EventDispatchers|CreateEvent :!event Custom|OnBigSpawnerDestroyed))
                (Default|BindEventtoOnSpawnedAnActor _asbga_continuous_bug_spawner_bug_biome (EventDispatchers|CreateEvent :!event Custom|OnNearbySpawnerSpawnedActor)))
              (:False
                (bind _asbga_continuous_bug_spawner_bug_biome_2 (Utilities|Casting|CastToBGA_ContinuousBugSpawner_BugBiome _array_element))
                (Game|BindEventtoOnDestroyed _asbga_continuous_bug_spawner_bug_biome_2 (EventDispatchers|CreateEvent :!event Custom|OnSmallSpawnerDestroyed))
                (Default|BindEventtoOnSpawnedAnActor _asbga_continuous_bug_spawner_bug_biome_2 (EventDispatchers|CreateEvent :!event Custom|OnNearbySpawnerSpawnedActor))))))))
    (:then_2
      (bind _outactors_2 (Collision|SphereOverlapActors (Transformation|GetActorLocation) (Variables|Default|GetGoldenRewardCheckRadius) (Utilities|Array|MakeArray "ObjectTypeQuery1" "ObjectTypeQuery2") "/BugBiomeGameplay/Gameplay/Environment/BP_BossSpire.BP_BossSpire_C"))
      (bind _output_get (Variables|Default|SetBossSpires _outactors_2))
      (for _array_element_2 _output_get
        (bind _asbp_boss_spire (Utilities|Casting|CastToBP_BossSpire _array_element_2))
        (Variables|/BugBiomeGameplay/Gameplay/Environment/BP_BossSpire.BP_BossSpire_C|SetActive true _asbp_boss_spire)))))

(event Custom|CheckForBossSpawn
  (if (>= (Variables|Default|GetBossSummonProgress) 1.0)
    (Utilities|FlowControl|DoOnce false
      (:Completed
        (CallFunction|SpawnBoss)))))

(event Custom|OnBossSpawned
  (params (SpawnedActor :type Actor%20Object%20Reference))
  (bind _asbuilding_gameplay_actor (Utilities|Casting|CastToBuildingGameplayActor SpawnedActor))
  (bind _output_get (Variables|Default|SetTidalCraneBoss _asbuilding_gameplay_actor))
  (EventDispatchers|BindEventtoOnDied _output_get (delegate Custom|OnBossDeath))
  (Class|AthenaSpecialActorComponent|SetMiniMapIcon (Variables|Default|GetAthenaSpecialActor) (Widget|Brush|MakeBrushfromMaterial (Variables|Default|GetBossMinimapIcon))))

(event Custom|OnSmallSpawnerDestroyed
  (params (DestroyedActor :type Actor%20Object%20Reference))
  ; Network progress and send events
  (CallFunction|AddProgress (ScalableFloat|GetValueatLevel (Variables|Default|GetSF_ProgressFromSmallSpawner)))
  (CallFunction|CheckforBossSpawn))

(event Custom|OnBossDeath
  (params (DamagedActor :type Actor%20Object%20Reference) (Damage :type Float%20%28single-precision%29) (InstigatedBy :type Controller%20Object%20Reference) (DamageCauser :type Actor%20Object%20Reference) (HitLocation :type Vector) (FHitComponent :type Primitive%20Component%20Object%20Reference) (BoneName :type Name) (Momentum :type Vector))
  (Variables|Default|SetIsEventFinished true)
  (Components|Activation|SetActive (Variables|Default|GetAthenaSpecialActor) false false)
  (for _array_element (Variables|Default|GetBossSpires)
    (Utilities|IsValid _array_element
      (:Is%20Valid
        (bind _asbp_boss_spire (Utilities|Casting|CastToBP_BossSpire _array_element))
        (Variables|/BugBiomeGameplay/Gameplay/Environment/BP_BossSpire.BP_BossSpire_C|SetActive false _asbp_boss_spire)))))

(event Custom|OnBigSpawnerDestroyed
  (params (DestroyedActor :type Actor%20Object%20Reference))
  ; Network progress and send events
  (CallFunction|AddProgress (ScalableFloat|GetValueatLevel (Variables|Default|GetSF_ProgressFromBigSpawner)))
  (CallFunction|CheckforBossSpawn)
  (if (not (>= (Variables|Default|GetBossSummonProgress) 1.0))
    (CallFunction|BossRoarSmallMulti)))

(event Custom|OnGoldenRewardOpened
  ; Network progress and send events
  (CallFunction|AddProgress (ScalableFloat|GetValueatLevel (Variables|Default|GetSF_ProgressFromRewardOpened)))
  (CallFunction|CheckforBossSpawn)
  (if (not (>= (Variables|Default|GetBossSummonProgress) 1.0))
    (CallFunction|BossRoarMulti)))

(event Custom|UpdatePotentialOverlapsAfterInit
  ; Notify UI of players in zone
  (Networking|CanExecuteCosmeticEvents
    (:True
      (Utilities|FlowControl|DeltaGateWildcard (Variables|Default|GetDidOverlapEventsSend) true
        (:Success
          (for _array_element (Collision|GetOverlappingActors (Variables|Default|GetPlayerEntranceVolume) 0 :!class /Script/Engine.PrimitiveComponent)
            (GameplayEventRouter|BroadcastStatefulEvent (GameplayEventRouter|GetGlobalEventRouter) (Utilities|Struct|MakeSTRUCTEventBugBiomePlayerEnteredExited true (Utilities|ResolveSoftReference self)) _array_element)))))))

(event OnComponentBeginOverlap(PlayerEntranceVolume)
  (params (OverlappedComponent :type Primitive%20Component%20Object%20Reference) (OtherActor :type Actor%20Object%20Reference) (OtherComp :type Primitive%20Component%20Object%20Reference) (OtherBodyIndex :type Integer) (bFromSweep :type Boolean) (SweepResult :type Hit%20Result%20Structure%20%28by%20ref%29))
  ; Notify UI of players in zone
  (Networking|CanExecuteCosmeticEvents
    (:True
      (GameplayEventRouter|BroadcastStatefulEvent (GameplayEventRouter|GetGlobalEventRouter) (Utilities|Struct|MakeSTRUCTEventBugBiomePlayerEnteredExited true (Utilities|ResolveSoftReference self)) OtherActor)
      (Variables|Default|SetDidOverlapEventsSend true))))

(event OnComponentEndOverlap(PlayerEntranceVolume)
  (params (OverlappedComponent :type Primitive%20Component%20Object%20Reference) (OtherActor :type Actor%20Object%20Reference) (OtherComp :type Primitive%20Component%20Object%20Reference) (OtherBodyIndex :type Integer))
  ; Notify UI of players in zone
  (Networking|CanExecuteCosmeticEvents
    (:True
      (GameplayEventRouter|BroadcastStatefulEvent (GameplayEventRouter|GetGlobalEventRouter) (Utilities|Struct|MakeSTRUCTEventBugBiomePlayerEnteredExited false (Utilities|ResolveSoftReference self)) OtherActor))))

(event Custom|BossRoar_Multi
  (bind _returnvalue (Transformation|GetActorLocation))
  (Audio|PlaySoundatLocation "/TidalCrane_Swarm_Boss/Assets/Sound/SC_QueenBug_Underground_Roar.SC_QueenBug_Underground_Roar" _returnvalue "0, 0, 0" 3.0)
  (Camera|PlayWorldCameraShake "/ProtoSwarm/Blueprints/EnvironmentalGameplay/B_CameraShake_Boss_Roar.B_CameraShake_Boss_Roar_C" _returnvalue (/ (Variables|Default|GetGoldenRewardCheckRadius) 3.0) (Variables|Default|GetGoldenRewardCheckRadius)))

(event Custom|BossRoarSmall_Multi
  (bind _returnvalue (Transformation|GetActorLocation))
  (Audio|PlaySoundatLocation "/TidalCrane_Swarm_Boss/Assets/Sound/SC_QueenBug_Underground_Roar.SC_QueenBug_Underground_Roar" _returnvalue "0, 0, 0" 1.5)
  (Camera|PlayWorldCameraShake "/ProtoSwarm/Blueprints/EnvironmentalGameplay/B_CameraShake_Boss_Roar_Small.B_CameraShake_Boss_Roar_Small_C" _returnvalue (Variables|Default|GetGoldenRewardCheckRadius) (/ (Variables|Default|GetGoldenRewardCheckRadius) 3.0)))

(event Custom|OnNearbySpawnerSpawnedActor
  (params (Spawned%20Actor :type Actor%20Object%20Reference))
  (bind _asfort_aipawn (Utilities|Casting|CastToFortAIPawn Spawned%20Actor))
  (EventDispatchers|BindEventtoOnDied _asfort_aipawn (delegate Custom|OnNearbyBugKilled)))

(event Custom|OnNearbyBugKilled
  (params (DamagedActor :type Actor%20Object%20Reference) (Damage :type Float%20%28single-precision%29) (InstigatedBy :type Controller%20Object%20Reference) (DamageCauser :type Actor%20Object%20Reference) (HitLocation :type Vector) (FHitComponent :type Primitive%20Component%20Object%20Reference) (BoneName :type Name) (Momentum :type Vector))
  (bind _returnvalue (GameplayTags|HasMatchingGameplayTag(Message) DamagedActor "(TagName=\"NPC.ProtoSwarm.MediumCreature\")"))
  (CallFunction|AddProgress (* (ScalableFloat|GetValueatLevel (select _returnvalue (Variables|Default|GetSF_ProgressFromBigBug) (Variables|Default|GetSF_ProgressFromSmallBug))) (Variables|Default|GetTeamSizeProgressMultiplier)))
  (CallFunction|CheckforBossSpawn))

(fn AddProgress
  (params (AddedProgress :type Float%20%28double-precision%29))
  (Networking|SwitchHasAuthority
    (:Authority
      (Variables|Default|SetBossSummonProgress (+ (Variables|Default|GetBossSummonProgress) AddedProgress)))))

(fn GetPOITag
  (return (Variables|Default|GetPOILocation)))

(fn IsThreatHigh
  (return (not (Variables|Default|GetIsEventFinished))))

(fn OnRep_BossSummonProgress
  (Default|CallOnBossSummonProgressMade (Variables|Default|GetBossSummonProgress)))

(fn OnRep_IsEventFinished
  (Variables|Default|GetIsEventFinished
    (:then
      (Default|CallOnEventFinished))
    (:else)))

(fn OnRep_TidalCraneBoss
  (Default|CallOnBossSet (Variables|Default|GetTidalCraneBoss)))

(fn SpawnBoss
  (bind _bossspawnpoints (Variables|Default|GetBossSpawnPoints))
  (bind _outitem (Utilities|Array|RandomArrayItem _bossspawnpoints))
  (bind _returnvalue (Class|FortAthenaLivingWorldManager|GetLivingWorldManager))
  (Class|FortAthenaLivingWorldPointProviderInterface|EnablePointProvider_DEPRECATED _outitem)
  (Class|FortAthenaLivingWorldManager|UnregisterDensityReservoir _returnvalue self)
  (Class|FortAthenaLivingWorldManager|RequestEventGeneration _returnvalue)
  (EventDispatchers|BindEventtoOnActorSpawned _outitem (EventDispatchers|CreateEvent :!event OnBossSpawned)))

(fn ConstructionScript
  (Rendering|Debug|DrawDebugSphere (Transformation|GetActorLocation) (Variables|Default|GetGoldenRewardCheckRadius) 20 "(R=0.026840,G=1.000000,B=0.000000,A=1.000000)" 2.0 2e+01)
  (Components|Sphere|SetSphereRadius (Variables|Default|GetPlayerEntranceVolume) (Variables|Default|GetGoldenRewardCheckRadius) true))
