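; Randomizer for the Tidal Crane Swarm Boss that decides which placed queen spawners remain in the match.
; Once playlist data is ready, it always shortlists Hive, then uses hotfixable weighted rolls for Utopia City versus Mask Maker and Tunnel versus Ranger Ruin, destroying each loser.
; An invalid Tunnel spawner falls back to Ranger Ruin.
; It then destroys random shortlisted spawners, by the maximum queen count minus the clamped hotfix count.
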(event EventBeginPlay
  (bind _validspawnersarray (Variables|Internal|GetValidSpawnersArray))
  (Class|FortAsyncActionCurrentPlaylistReady|PlaylistDataReadyAsync
    (:OnReady
      (Utilities|FlowControl|Sequence
        (:then_0
          (Utilities|Array|Add (Variables|Internal|GetValidSpawnersArray) (Variables|PlacedinWorld|GetHiveSpawner)))
        (:then_1
          (if (Math|Random|RandomBoolwithWeight (ScalableFloat|GetValueatLevel (Variables|Hotfix|GetChanceToSpawnUtopia)))
            (Actor|DestroyActor (Variables|PlacedinWorld|GetMaskMakerSpawner))
            (Utilities|Array|Add (Variables|Internal|GetValidSpawnersArray) (Variables|PlacedinWorld|GetUtopiaCitySpawner))
            (else
              (Actor|DestroyActor (Variables|PlacedinWorld|GetUtopiaCitySpawner))
              (Utilities|Array|Add (Variables|Internal|GetValidSpawnersArray) (Variables|PlacedinWorld|GetMaskMakerSpawner)))))
        (:then_2
          (Utilities|IsValid (Variables|PlacedinWorld|GetTunnelSpawner)
            (:Is%20Valid
              (if (Math|Random|RandomBoolwithWeight (ScalableFloat|GetValueatLevel (Variables|Hotfix|GetChanceToSpawnTunnel)))
                (Actor|DestroyActor (Variables|PlacedinWorld|GetRangerRuinSpawner))
                (Utilities|Array|Add (Variables|Internal|GetValidSpawnersArray) (Variables|PlacedinWorld|GetTunnelSpawner))
                (else
                  (Actor|DestroyActor (Variables|PlacedinWorld|GetTunnelSpawner))
                  (Utilities|Array|Add (Variables|Internal|GetValidSpawnersArray) (Variables|PlacedinWorld|GetRangerRuinSpawner)))))
            (:Is%20Not%20Valid
              (Utilities|Array|Add (Variables|Internal|GetValidSpawnersArray) (Variables|PlacedinWorld|GetRangerRuinSpawner)))) ; Tunnel won't be available till .20 releases. Pick Ranger Ruin's if Tunnel's not valid
        )
        (:then_3
          (for _ (range 1 (+ (- (Math|Float|Truncate (Variables|Internal|GetMaxNumQueensToSpawn)) (Math|Float|Truncate (Math|Float|Clamp(Float) (ScalableFloat|GetValueatLevel (Variables|Hotfix|GetNumQueensToSpawn)) 0.0 (Variables|Internal|GetMaxNumQueensToSpawn)))) 1))
            (bind _element (ArrayMacros|GetRandomArrayElement _validspawnersarray)
              (:%20%20
                (Actor|DestroyActor _element)
                (Utilities|Array|RemoveItem _validspawnersarray _element)))))))))

(fn ConstructionScript)
