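; Actor blueprint marking reposition locations for the Tidal Crane boss, able to spawn a boss base at its own transform.
; On BeginPlay, if AllowSpawnBase is set, it spawns either BossBaseClassNoRVT or BossBaseClass, chosen by the IsInTunnel flag.
; The spawn always succeeds regardless of collision and multiplies scale with the root; the construction script is empty.
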(event EventBeginPlay
  (if (Variables|Default|GetAllowSpawnBase)
    (Game|SpawnActorfromClass (select (Variables|Default|GetIsInTunnel) (Variables|Default|GetBossBaseClassNoRVT) (Variables|Default|GetBossBaseClass)) (Transformation|GetActorTransform) "AlwaysSpawn" "MultiplyWithRoot" 0 0)))

(fn ConstructionScript)
